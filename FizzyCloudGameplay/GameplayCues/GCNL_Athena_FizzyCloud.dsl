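; Looping gameplay cue for the FizzyCloud heal spray that plays healing audio on the affected player pawn.
; On loop start it fades out any existing heal audio, attaches a 1P or 3P loop based on whether the pawn is locally viewed, and starts a sound indicator if the target is the instigator's own pawn.
; On removal it plays an attached 1P or 3P spray-stop sound, fades out the loop and stops the indicator.
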(event EventOnLoopingStart(EitherVFX-preferthisversion)
  (params (MyTarget :type Actor%20Object%20Reference) (Parameters :type Gameplay%20Cue%20Parameters%20Structure%20%28by%20ref%29) (ParticleComponents :type Array%20of%20FXSystem%20Component%20Object%20References) (AudioComponents :type Array%20of%20Audio%20Component%20Object%20References))
  (bind (_normalizedmagnitude _rawmagnitude _effectcontext _matchedtagname _originaltag _aggregatedsourcetags _aggregatedtargettags _location _normal _instigator) (Ability|GameplayCue|BreakGameplayCueParameters Parameters))
  (bind _asfort_player_pawn (Utilities|Casting|CastToFortPlayerPawn MyTarget))
  (Variables|Default|SetPlayerPawn _asfort_player_pawn)
  (Sound|StartHealAudio)
  (bind _asplayer_state (Utilities|Casting|CastToPlayerState _instigator))
  (if (== MyTarget (PlayerState|GetPawn _asplayer_state))
    (SoundIndicator|StartTargetedSoundIndicatorLooping (Variables|Default|GetFortSoundIndicator_OnHeal) MyTarget "/FizzyCloudGameplay/GameplayCues/Assets/Audio/FizzyCloud_HealSpray_Heal_3P_Loop_Healing_Cue.FizzyCloud_HealSpray_Heal_3P_Loop_Healing_Cue" 1.0 0)))

(event EventOnRemoval(EitherVFX-preferthisversion)
  (params (MyTarget :type Actor%20Object%20Reference) (Parameters :type Gameplay%20Cue%20Parameters%20Structure%20%28by%20ref%29) (ParticleComponents :type Array%20of%20FXSystem%20Component%20Object%20References) (AudioComponents :type Array%20of%20Audio%20Component%20Object%20References) (BurstCameraShakeInstance :type Camera%20Shake%20Base%20Object%20Reference) (BurstDecalInstance :type Decal%20Actor%20Object%20Reference))
  (Sound|StopHealAudio)
  (SoundIndicator|StopSoundIndicatorLooping (Variables|Default|GetFortSoundIndicator_OnHeal)))

(fn StartHealAudio
  (bind _healaudiocomponent (Variables|Default|GetHealAudioComponent)
    (:then
      (Audio|Components|Audio|FadeOut _healaudiocomponent 0.25 0.0 "Linear" :!class /Script/Engine.AudioComponent))
    (:else
      _))
  (bind _returnvalue (Audio|SpawnSoundAttached (select (Pawn|IsLocallyViewed (Variables|Default|GetPlayerPawn)) (Variables|Sound|GetHealSound1p) (Variables|Sound|GetHealSound3p)) (Variables|/Script/Engine.Actor|GetRootComponent (Variables|Default|GetPlayerPawn))))
  (Variables|Default|SetHealAudioComponent _returnvalue))

(fn StopHealAudio
  (bind _playerpawn (Variables|Default|GetPlayerPawn)
    (:then
      (Audio|SpawnSoundAttached (select (Pawn|IsLocallyViewed _playerpawn) (Variables|Default|GetSprayStopSound1p) (Variables|Default|GetSprayStopSound3p)) (Variables|/Script/Engine.Actor|GetRootComponent _playerpawn)))
    (:else
      _))
  (bind _healaudiocomponent (Variables|Default|GetHealAudioComponent)
    (:then
      (Audio|Components|Audio|FadeOut _healaudiocomponent 0.25 0.0 "Linear" :!class /Script/Engine.AudioComponent))
    (:else)))

(fn ConstructionScript)
